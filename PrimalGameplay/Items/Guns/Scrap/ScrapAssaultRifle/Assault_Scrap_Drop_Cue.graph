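# Sound cue for dropping the scrap assault rifle, under the loot-collect sound class.
# A random node picks with equal weight among four generic assault rifle drop waves, played with boosted volume and weapon-drop attenuation.
# A concurrency override allows one instance and blocks new plays, taking the place of the assigned stop-oldest concurrency set.
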SoundCue /PrimalGameplay/Items/Guns/Scrap/ScrapAssaultRifle/Assault_Scrap_Drop_Cue
root=Random_1
nodes [
 {id=Random_1, children [WavePlayer_0,WavePlayer_3,WavePlayer_4,WavePlayer_5], properties {weights [1,1,1,1]}}
 {id=WavePlayer_0, properties {soundWaveAssetPtr=/Game/Sounds/Fort_UI_Sounds/Gameplay/Weapon_Generic_Drop_AR_04}}
 {id=WavePlayer_3, properties {soundWaveAssetPtr=/Game/Sounds/Fort_UI_Sounds/Gameplay/Weapon_Generic_Drop_AR_01}}
 {id=WavePlayer_4, properties {soundWaveAssetPtr=/Game/Sounds/Fort_UI_Sounds/Gameplay/Weapon_Generic_Drop_AR_02}}
 {id=WavePlayer_5, properties {soundWaveAssetPtr=/Game/Sounds/Fort_UI_Sounds/Gameplay/Weapon_Generic_Drop_AR_03}}
]
properties {
 attenuationSettings=&Fort_Weapon_Drop_Att
 bOverrideConcurrency=true
 maxDistance=3000
 soundClassObject=&Fort_Loot_Collect
 volumeMultiplier=1.5
 concurrencyOverrides {maxCount=1, resolutionRule=PreventNew}
 concurrencySet [&Weapon_Drop_StopOldest]
}
