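# Sound cue for the close first-person firing sound of the scrap assault rifle.
# It picks one of four equally weighted close-fire waves at random, then passes it through a modulator for variation.
# It uses first-person projectile attenuation, a ranged-weapon first-person sound class and a shared first-person gunfire concurrency set.
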SoundCue /PrimalGameplay/Items/Guns/Scrap/ScrapAssaultRifle/Assault_Scrap_Fire_1P_Cue
root=Modulator_1
nodes [
 {id=Modulator_1, children [Random_0]}
 {id=Random_0, children [WavePlayer_0,WavePlayer_1,WavePlayer_2,WavePlayer_3], properties {weights [1,1,1,1]}}
 {id=WavePlayer_0, properties {soundWaveAssetPtr=&/Assault_Scrap_Fire_Close_01}}
 {id=WavePlayer_1, properties {soundWaveAssetPtr=&/Assault_Scrap_Fire_Close_02}}
 {id=WavePlayer_2, properties {soundWaveAssetPtr=&/Assault_Scrap_Fire_Close_03}}
 {id=WavePlayer_3, properties {soundWaveAssetPtr=&/Assault_Scrap_Fire_Close_04}}
]
properties {
 attenuationSettings=&Fort_Weapon_Projectile_FirstPerson_Athena_Att
 maxDistance=2097152
 soundClassObject=&Fort_WeaponRanged_Fire_1P_PSM
 volumeMultiplier=0.7
 attenuationOverrides {distanceAlgorithm=NaturalSound, falloffDistance=5000, attenuationShapeExtents {x=1000}}
 concurrencyOverrides {bLimitToOwner=true, maxCount=3, resolutionRule=StopOldest}
 concurrencySet [&Weapon_Gunfire_FirstPerson_Conc]
}
