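# Sound cue for the low-ammo sound of the Scrap Assault Rifle, routed to the degraded ranged-weapon low-ammo sound class.
# A single wave player feeds a modulator whose pitch and volume ranges are fixed at unity, so it adds no variation.
# The cue slightly boosts volume, enables a concurrency override, and carries natural-sound distance attenuation settings.
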SoundCue /PrimalGameplay/Items/Guns/Scrap/ScrapAssaultRifle/Assault_Scrap_Low_Ammo_Cue
root=Modulator_3
nodes [
 {id=Modulator_3, children [WavePlayer_1], properties {pitchMax=1, pitchMin=1, volumeMax=1, volumeMin=1}}
 {id=WavePlayer_1, properties {soundWaveAssetPtr=&/Assault_Scrap_Low_Ammo_01}}
]
properties {
 bOverrideConcurrency=true
 maxDistance=2097152
 soundClassObject=/Game/Sounds/FortSoundClassesAndModes/Fort_WeaponRanged_LowAmmo_Degraded_PSM
 volumeMultiplier=1.2
 attenuationOverrides {distanceAlgorithm=NaturalSound, falloffDistance=5000, attenuationShapeExtents {x=1000}}
}
