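# Sound cue for picking up the scrap assault rifle, routed to the loot-collect sound class.
# A single wave player plays the pickup sample through a modulator that slightly randomizes pitch for variation.
# It uses shared weapon pickup attenuation at reduced volume, and overrides concurrency so only one instance plays at a time, rejecting new requests while one is active.
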SoundCue /PrimalGameplay/Items/Guns/Scrap/ScrapAssaultRifle/Assault_Scrap_Pickup_Cue
root=Modulator_0
nodes [
 {id=Modulator_0, children [WavePlayer_4], properties {pitchMax=1.1, pitchMin=0.9}}
 {id=WavePlayer_4, properties {soundWaveAssetPtr=&/Assault_Scrap_Pickup_01}}
]
properties {
 attenuationSettings=&Fort_Weapon_Pickup_Att
 bOverrideConcurrency=true
 maxDistance=3500
 soundClassObject=&Fort_Loot_Collect
 volumeMultiplier=0.478
 concurrencyOverrides {maxCount=1, resolutionRule=PreventNew}
}
